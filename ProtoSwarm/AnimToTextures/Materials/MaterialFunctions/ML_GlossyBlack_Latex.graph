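# Material function that outputs a dark latex surface layer as material attributes, built from tiling Mannequin rubber textures.
# A Tiling input scales the UVs shared by the diffuse and normal samples, both using shared wrap samplers.
# Base color is the diffuse multiplied by a Tint input that defaults to dark grey.
# The diffuse red channel also drives roughness within a narrow range, and the normal map feeds Normal directly.
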MaterialFunction /ProtoSwarm/AnimToTextures/Materials/MaterialFunctions/ML_GlossyBlack_Latex
graph {nodes [
  n1 = FunctionInput() {bUsePreviewValueAsDefault=true, inputName=Tiling, inputType=FunctionInput_Scalar, previewValue=<50 0 0 1>}
  n2 = Multiply(A=@n1, B=TextureCoordinate())
  n3 = TextureSample(UVs=@n2) {samplerSource=SSM_Wrap_WorldGroupSettings, texture=/AnimToTexture/Characters/Mannequin/Textures/T_ML_Rubber_Blue_01_D}
  n4 = FunctionInput() {bUsePreviewValueAsDefault=true, inputName=Tint, previewValue=<0.225 0.225 0.225 1.5>}
  n5 = TextureSample(UVs=@n2) {
   samplerSource=SSM_Wrap_WorldGroupSettings
   samplerType=SAMPLERTYPE_Normal
   texture=/AnimToTexture/Characters/Mannequin/Textures/T_ML_Rubber_Blue_01_N
  }
  FunctionOutput(None=MakeMaterialAttributes!(
    BaseColor=Multiply(A=@n3.RGB, B=@n4)
    Metallic
    Specular
    Roughness=LinearInterpolate(Alpha=@n3.R) {constA=0.55, constB=0.75}
    Anisotropy
    EmissiveColor
    Opacity
    OpacityMask
    Normal=@n5.RGB
    Tangent
    WorldPositionOffset
    SubsurfaceColor
    ClearCoat
    ClearCoatRoughness
    AmbientOcclusion
    Refraction
    CustomizedUVs_0
    CustomizedUVs_1
    CustomizedUVs_2
    CustomizedUVs_3
    CustomizedUVs_4
    CustomizedUVs_5
    CustomizedUVs_6
    CustomizedUVs_7
    PixelDepthOffset
    ShadingModel
    Displacement
   )) {sortPriority=1}
 ]}
